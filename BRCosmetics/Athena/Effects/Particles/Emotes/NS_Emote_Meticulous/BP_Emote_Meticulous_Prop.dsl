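; Prop actor for the Meticulous emote in Battle Royale that adds a camera lens effect while the emote prop exists.
; Shortly after BeginPlay it casts its attach parent to a FortPawn and, if that pawn is locally viewed, spawns the Meticulous camera lens effect on it and stores the reference.
; On EndPlay it deactivates the stored effect's primary particle component rather than destroying it.
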(event EventBeginPlay
  (Utilities|FlowControl|Delay 0.1)
  (bind _asfort_pawn (Utilities|Casting|CastToFortPawn (Actor|GetAttachParentActor self)))
  (if (Pawn|IsLocallyViewed _asfort_pawn)
    (bind _returnvalue (Feedback|SpawnGenericCameraLensEffect _asfort_pawn "/BRCosmetics/Athena/Effects/Particles/Emotes/NS_Emote_Meticulous/BP_Emote_Meticulous_CameraLens.BP_Emote_Meticulous_CameraLens_C"))
    (Variables|Default|SetCameraLensFX _returnvalue)))

(event EventEndPlay
  (params (EndPlayReason :type EEndPlayReason%20Enum :default "Destroyed"))
  (bind _cameralensfx (Variables|Default|GetCameraLensFX))
  (Utilities|IsValid (Utilities|ToObject(Interface) _cameralensfx)
    (:Is%20Valid
      (Components|Activation|Deactivate (CameraLensEffect|GetPrimaryParticleComponent _cameralensfx)))))

(fn ConstructionScript)
