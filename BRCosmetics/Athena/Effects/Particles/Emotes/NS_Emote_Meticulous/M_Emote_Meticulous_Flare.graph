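# Translucent, unlit Niagara sprite material that draws a flare effect for the Meticulous emote.
# The scaled, particle-rotated flare texture blends toward a curve-atlas color and panning radial noise, then gets desaturation and particle color for emissive.
# Alpha uses two rotating masks times a sphere mask, or a flare channel via static switch, shaped by view-angle fade, depth fade and particle alpha.
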Material /BRCosmetics/Athena/Effects/Particles/Emotes/NS_Emote_Meticulous/M_Emote_Meticulous_Flare
graph {
 properties {
  bUsedWithNiagaraSprites=true
  blendMode=BLEND_Translucent
  pixelDepthOffsetMode=PDOM_Legacy
  shadingModel=MSM_Unlit
  twoSided=true
  usedShadingModels=Unlit
 }
 nodes [
  n1 = ScalarParameter() {defaultValue=0.85, group=Flare, parameterName=Flare Scale}
  n2 = DynamicParameter() {defaultValue=<rgba 1 1 1 0>, paramNames [Flare Rotation, Mask Rotation Rate, Param3, Random]}
  n3 = TextureSampleParameter2D!(
   UVs=MaterialFunctionCall(
    UVs=MaterialFunctionCall(Texture Scale=@n1) {materialFunction=&ScaleUVsByCenter}.UVs
    "Rotation Angle (0-1)"=@n2.Flare Rotation
   ) {materialFunction=&CustomRotator}.Rotated Values
   Level
   Apply View MipBias
  ) {
   group=Flare
   mipValueMode=TMVM_MipLevel
   parameterName=Flare Texture
   samplerSource=SSM_Clamp_WorldGroupSettings
   texture=&/T_Emote_Meticulous_Flare_01
  }
  n4 = CurveAtlasRowParameter(CurveTime=@n3.G) {
   atlas=/Game/Athena/LinearColorCurves/A_Atlas_Turtleneck_Glider
   curve=/Game/Athena/LinearColorCurves/Curve_Emote_Meticulous_01
   defaultValue=25
   group=Color
   parameterName=Curve
  }
  n5 = LinearInterpolate(A=@n3.RGB, B=@n4, Alpha=ScalarParameter() {group=Color, parameterName=Lerp Color Curve}) {constAlpha=0}
  n6 = MaterialFunctionCall() {materialFunction=&VectorToRadialValue}
  n7 = VectorParameter() {group=Noise, parameterName=Noise01_UV, defaultValue=<rgba 5 0.2 0.5 -0.5>}
  n8 = AppendVector(A=@n7.R, B=@n7.R)
  n9 = TextureSample!(UVs=Panner(Coordinate=Multiply(A=@n6.Radial Coordinates, B=@n8), Speed=AppendVector(A=@n7.B, B=@n7.B)), Tex, Level, Apply View MipBias) {
   mipValueMode=TMVM_MipLevel
   samplerType=SAMPLERTYPE_LinearColor
   texture=/Game/Effects/Fort_Effects/World/Enviroment/Lightning/lightningPainting
  }
  n10 = VectorParameter() {group=Noise, parameterName=Noise02_UV, defaultValue=<rgba 4 0.5 1 -0.5>}
  n11 = Multiply(A=@n6.Radial Coordinates, B=AppendVector(A=@n10.R, B=@n10.R))
  n12 = AppendVector(A=@n10.B, B=@n10.B)
  n13 = Multiply(
   A=@n9.G
   B=TextureSample(UVs=Panner(Coordinate=@n11, Speed=@n12)) {mipValueMode=TMVM_MipLevel, texture=/Game/Effects/Fort_Effects/Textures/Tile/T_Tile_Texture_Cartoon_Flames}.G
  )
  n14 = LinearInterpolate(
   A=@n5
   B=Multiply(A=Multiply(A=@n13, B=@n5), B=ScalarParameter() {defaultValue=5, group=Noise, parameterName=Noise Intensity}) {constB=2}
   Alpha=ScalarParameter() {defaultValue=0.3, group=Noise, parameterName=Lerp Radial Noise}
  ) {constAlpha=0.2}
  n15 = ScalarParameter() {group=Color, parameterName=Desaturation}
  n16 = ParticleColor()
  MaterialFunctionCall(Emissive=Multiply(A=Desaturation(None=@n14, Fraction=@n15), B=@n16.RGB)) {materialFunction=&MF_FX_EmissiveAdjustments}
  n17 = ScalarParameter() {defaultValue=-1, group=Mask, parameterName=Mask01_Rotation}
  n18 = Multiply(A=@n2.Mask Rotation Rate, B=Time())
  n19 = TextureSampleParameter2D(UVs=Rotator(Time=Multiply(A=@n17, B=@n18)) {speed=1}) {
   group=Mask
   mipValueMode=TMVM_MipLevel
   parameterName=Mask Texture 01
   samplerSource=SSM_Clamp_WorldGroupSettings
   texture=&T_Emote_Meticulous_Flare_01
  }
  n20 = ScalarParameter() {defaultValue=1.4, group=Mask, parameterName=Mask02_Rotation}
  n21 = TextureSampleParameter2D(UVs=Rotator(Time=Multiply(A=@n18, B=@n20)) {speed=1}) {
   group=Mask
   mipValueMode=TMVM_MipLevel
   parameterName=Mask Texture 02
   samplerSource=SSM_Clamp_WorldGroupSettings
   texture=&/T_Emote_Meticulous_Flare_02
  }
  n22 = TextureCoordinate()
  n23 = Constant() {r=0.5}
  n24 = Constant() {r=0.5}
  n25 = Constant()
  n26 = StaticSwitchParameter(
   True=Multiply(A=Multiply(A=@n19.A, B=@n21.A), B=SphereMask(A=@n22, B=@n23, Radius=@n24, Hardness=@n25))
   False=ChannelMaskParameter(None=@n3.RGBA) {group=Alpha, maskChannel=Alpha, parameterName=Alpha Channel, defaultValue=<rgba 0 0 0 1>}.R
  ) {defaultValue=true, group=Mask, parameterName=Use Mask}
  n27 = Multiply(A=@n26, B=MaterialFunctionCall() {materialFunction=&FadeBasedOnViewAngle}.Result)
  n28 = ScalarParameter() {defaultValue=1, group=Alpha, parameterName=Use Angle Fade}
  Multiply(
   A=@n16.A
   B=DepthFade(
    Opacity=LinearInterpolate(A=@n26, B=@n27, Alpha=@n28)
    FadeDistance=ScalarParameter() {defaultValue=50, group=Alpha, parameterName=FadeDistance}
   )
  )
 ]
}
