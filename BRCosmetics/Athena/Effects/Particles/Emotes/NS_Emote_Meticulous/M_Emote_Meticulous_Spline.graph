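# Translucent unlit, two-sided material for Niagara mesh particles in the Meticulous emote, drawing spline ribbons widened by SplineThicken through a per-particle width multiplier.
# Emissive color is tinted by particle color.
# Opacity blends a sphere or line mask, eroded by panning cell noise with per-particle random offset and dissolve, then depth-faded except on mobile.
# Dissolve tiling and pan speed reuse their U values for V, leaving V controls unused.
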Material /BRCosmetics/Athena/Effects/Particles/Emotes/NS_Emote_Meticulous/M_Emote_Meticulous_Spline
graph {
 properties {
  bUsedWithNiagaraMeshParticles=true
  blendMode=BLEND_Translucent
  pixelDepthOffsetMode=PDOM_Legacy
  shadingModel=MSM_Unlit
  twoSided=true
  usedShadingModels=Unlit
 }
 nodes [
  n1 = DynamicParameter() {defaultValue=<rgba 0 1 1 1>, paramNames [Random, Dissolve, Param3, SplineWidth Mult]}
  n2 = Multiply(A=@n1.SplineWidth Mult, B=ScalarParameter() {defaultValue=0.5, group=WPO, parameterName=SplineWidth})
  MaterialFunctionCall(WidthBase=@n2, WidthTip=@n2) {materialFunction=&SplineThicken}
  n3 = VectorParameter() {group=Emissive, parameterName=Emissive Color, defaultValue=<rgba 1 1 1 1>}
  n4 = ParticleColor()
  MaterialFunctionCall(Emissive=Multiply(A=@n3.RGB, B=@n4.RGB)) {materialFunction=&MF_FX_EmissiveAdjustments}
  n5 = ScalarParameter() {defaultValue=1, group=Mask, parameterName=SphereMask Thickness}
  n6 = Multiply(A=TextureCoordinate(), B=AppendVector(A=@n5, B=Constant() {r=1}))
  n7 = Constant() {r=0.5}
  n8 = AppendVector(A=Multiply(A=@n5, B=@n7), B=@n7)
  n9 = Constant() {r=0.5}
  n10 = ScalarParameter() {group=Mask, parameterName=SphereMask Hardness}
  n11 = MaterialFunctionCall(UV Channel=TextureCoordinate()) {materialFunction=&LinearGradient}
  n12 = OneMinus(None=@n11.UGradient)
  n13 = Multiply(
   A=Power(Base=Multiply(A=@n12, B=@n11.UGradient), Exp=ScalarParameter() {defaultValue=1, group=Mask, parameterName=LineMask Power}) {constExponent=1}
   B=ScalarParameter() {defaultValue=2, group=Mask, parameterName=LineMask Mult}
  )
  n14 = VectorParameter() {
   group=Dissolve
   parameterName=Dissolve UV Control
   channelNames {a=Dissolve_V_PanSpeed, b=Dissolve_U_PanSpeed, g=Dissolve_V_Tilling, r=Dissolve_U_Tilling}
   defaultValue=<rgba 0.2 2.5 0.5 0>
  }
  n15 = Multiply(A=TextureCoordinate(), B=AppendVector(A=@n14.Dissolve_U_Tilling, B=@n14.Dissolve_U_Tilling))
  n16 = AppendVector(A=@n14.Dissolve_U_PanSpeed, B=@n14.Dissolve_U_PanSpeed)
  n17 = AppendVector(A=ScalarParameter() {group=Dissolve, parameterName=Dissolve_U_Offset}, B=ScalarParameter() {group=Dissolve, parameterName=Dissolve_V_Offset})
  n18 = TextureSampleParameter2D(UVs=Add(A=Panner(Coordinate=@n15, Speed=@n16), B=Add(A=@n17, B=@n1.Random))) {
   group=Dissolve
   parameterName=Dissolve Tex
   samplerType=SAMPLERTYPE_Grayscale
   texture=/BRCosmetics/Weapons/FORT_Melee/Pickaxe_Cube_Ninja_Male/FX/T_CellNoise
  }
  n19 = ScalarParameter() {defaultValue=6, group=Dissolve, parameterName=Dissolve Edge Hardness, sliderMin=0.5}
  n20 = Multiply(A=Power(Base=@n18.G, Exp=ScalarParameter() {defaultValue=0.5, group=Dissolve, parameterName=Dissolve Power}), B=@n19)
  n21 = LinearInterpolate(A=@n19, Alpha=Multiply(A=ScalarParameter() {defaultValue=1, group=Dissolve, parameterName=Dissolve Dynamic Test}, B=@n1.Dissolve)) {constB=-1}
  n22 = Multiply(
   A=LinearInterpolate(
    A=SphereMask(A=@n6, B=@n8, Radius=@n9, Hardness=@n10)
    B=Saturate(None=@n13)
    Alpha=ScalarParameter() {defaultValue=1, group=Mask, parameterName=Mask Type}
   ) {constAlpha=0}
   B=Saturate(None=Subtract(A=@n20, B=@n21))
  )
  n23 = DepthFade(Opacity=@n22, FadeDistance=ScalarParameter() {defaultValue=100, group=Alpah, parameterName=Fade_Distance}) {fadeDistanceDefault=10}
  Multiply(A=@n4.A, B=Multiply(A=@n3.A, B=ShadingPathSwitch(Default=@n23, Deferred=@n23, Mobile=@n22) {desc=Migrated from FeatureLevelSwitch}))
 ]
}
